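# Control Rig for Kieran's face accessory part that corrects his back hair as the head moves relative to the upper spine.
# A spherical pose reader on the head, parented to spine_04, outputs a pose weight that is multiplied by the Detected Dual Pistols variable.
# The product weights two chained Modify Transforms that offset and squash hair_lower_area and rotate and shift base_hair_back.
# With that variable at its default of zero, the correction is inactive.
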ControlRigBlueprint /BRCosmetics/Characters/Player/Male/Medium/Bodies/M_MED_Kieran/Meshes/Parts/M_MED_Kieran_FaceAcc_CtrlRig
graphs [
 RigVMFunctionLibrary {links [], nodes []}
 RigVMModel {nodes [
   RigUnit_BeginExecution = &RigUnit_BeginExecution()
   n1 = &RigUnit_SphericalPoseReader(
    ActiveRegionScaleFactors=(PositiveWidth=3,NegativeWidth=0.2,PositiveHeight=6,NegativeHeight=6)
    ActiveRegionSize=0.1
    Debug=(bDrawDebug=True,bDraw2D=False,bDrawLocalAxes=False,DebugScale=25,DebugSegments=20,DebugThickness=0.25)
    DriverAxis=(X=1,Y=0,Z=0)
    DriverItem=(Type=Bone,Name="head")
    FalloffRegionScaleFactors=(PositiveWidth=-3,NegativeWidth=0.25,PositiveHeight=0,NegativeHeight=0)
    FalloffSize=0.25
    FlipHeightScaling=False
    FlipWidthScaling=False
    OptionalParentItem=(Type=Bone,Name="spine_04")
    RotationOffset=(X=90,Y=0,Z=-180)
    ExecutePin=@RigUnit_BeginExecution.ExecutePin
   )
   Multiply = &RigVMFunction_MathDoubleMul(A=@n1.OutputParam, B=RigVMVariableNode() {variable=Detected Dual Pistols}.Value)
   > &RigUnit_ModifyTransforms(
    ItemToModify=((Item=(Type=Bone,Name="hair_lower_area"),Transform=(Rotation=(X=0,Y=0,Z=0,W=1),Translation=(X=0,Y=0,Z=2.5),Scale3D=(X=1,Y=0.75,Z=0.75))))
    WeightMaximum=0.8
    WeightMinimum=0
    Weight=@Multiply.Result
   )
   > &RigUnit_ModifyTransforms(
    ItemToModify=((Item=(Type=Bone,Name="base_hair_back"),Transform=(Rotation=(X=0,Y=0,Z=0.1736,W=0.9848),Translation=(X=2,Y=0,Z=0),Scale3D=(X=1,Y=1,Z=1))))
    WeightMinimum=0
    Weight=@Multiply.Result
   )
  ]}
]
hierarchy [
 root {loc=<0 0 0>, rot=<0 0 0 1>}
 pelvis {parent=root, loc=<0 -0.7071 93.1191>, rot=<0.0127 0.707 -0.0127 -0.707>}
 spine_01 {parent=pelvis, loc=<6.7494 -0.6388 0>, rot=<0 0 0.034 -0.9994>}
 spine_02 {parent=spine_01, loc=<7.2518 -0.035 0>, rot=<0 0 0.0238 -0.9997>}
 spine_03 {parent=spine_02, loc=<6.0104 0 0>, rot=<0 0 0.0149 -0.9999>}
 spine_04 {parent=spine_03, loc=<8.1024 0 0>, rot=<0 0 -0.0132 -0.9999>}
 spine_05 {parent=spine_04, loc=<10.0146 0 0>, rot=<0 0 -0.0015 -1>}
 neck_01 {parent=spine_05, loc=<11.3138 0.1529 0>, rot=<0 0 -0.2003 -0.9797>}
 neck_02 {parent=neck_01, loc=<4.5627 0 0>, rot=<0 0 0.0504 -0.9987>}
 head {parent=neck_02, loc=<5.3914 0 0>, rot=<0 0 0.115 -0.9934>}
 base_hair_front {parent=head, loc=<5.095 5.7885 6.4529>, rot=<-0.0905 -0.6847 0.7127 0.1229>}
 hair_front_F_1 {parent=base_hair_front, loc=<0 0 0>, rot=<0 0 0 1>}
 hair_front_F_2 {parent=hair_front_F_1, loc=<3.3404 -0.4633 0>, rot=<0.0211 0.2984 -0.0673 -0.9518>}
 hair_front_F_3 {parent=hair_front_F_2, loc=<3.2878 0 0>, rot=<0 0 0 1>}
 base_hair_back {parent=head, loc=<6.9815 -7.1002 0.4145>, rot=<0 0 0 1>}
 hair_lower_area {parent=base_hair_back, loc=<-1.8854 -3.179 0.145>, rot=<0.0029 -0.7071 -0.0029 -0.7071>}
 hair_base_D_1 {parent=hair_lower_area, loc=<-0.1671 -4.6411 -2.2082>, rot=<-0.2937 -0.5487 0.567 -0.5397>}
 hair_base_D_2 {parent=hair_lower_area, loc=<-1.7695 -5.0377 -1.3545>, rot=<-0.2178 -0.583 0.6339 -0.4591>}
 hair_base_D_3 {parent=hair_lower_area, loc=<2.7196 -2.5069 0.5717>, rot=<-0.1088 -0.4951 0.4724 -0.721>}
 hair_base_E_1 {parent=hair_lower_area, loc=<-5.2138 2.644 -4.1101>, rot=<-0.4733 -0.3101 0.7688 -0.298>}
 hair_top_area {parent=base_hair_back, loc=<3.8707 -4.1843 -1.7758>, rot=<0.0029 -0.7071 -0.0029 -0.7071>}
 base_hair_base_A {parent=hair_top_area, loc=<-1.543 3.3994 3.8398>, rot=<-0.0029 0.7071 0.0029 -0.7071>}
 hair_base_A_1 {parent=base_hair_base_A, loc=<-0.1294 0.9906 1.4575>, rot=<-0.5109 0.1891 -0.6444 -0.5367>}
 hair_base_A_2 {parent=base_hair_base_A, loc=<0.2369 0.9146 -1.4253>, rot=<-0.3516 -0.6051 -0.508 -0.5022>}
 hair_base_A_3 {parent=base_hair_base_A, loc=<0.644 1.2335 0.4498>, rot=<-0.5445 0.017 -0.4405 -0.7136>}
 hair_base_A_4 {parent=base_hair_base_A, loc=<0.7659 0.1651 0.037>, rot=<-0.0532 0.0946 -0.1702 -0.9794>}
 hair_base_A_5 {parent=base_hair_base_A, loc=<-0.066 -1.0061 0.6679>, rot=<0.6577 -0.3092 0.4998 0.4712>}
 hair_base_A_6 {parent=base_hair_base_A, loc=<-1.5533 0.7352 -0.5516>, rot=<0.677 0.2775 0.6703 0.1241>}
 hair_base_B_1 {parent=hair_top_area, loc=<3.1406 -3.6288 1.0141>, rot=<0.0342 0.0829 0.3859 -0.9182>}
 hair_base_B_2 {parent=hair_top_area, loc=<3.6425 -3.9381 -1.3147>, rot=<-0.241 -0.5724 0.3033 -0.7227>}
 hair_base_B_3 {parent=hair_top_area, loc=<2.3413 -4.7144 -0.1376>, rot=<-0.1579 -0.4537 0.5407 -0.6906>}
 hair_base_C_1 {parent=base_hair_back, loc=<-1.6984 0.9941 -4.2223>, rot=<-0.048 -0.9453 -0.1025 -0.3058>}
 hair_base_C_2 {parent=base_hair_back, loc=<-1.1317 -0.2925 -4.6886>, rot=<0.102 -0.8598 0.054 -0.4975>}
]
variables [Detected Dual Pistols {type=double, default=0}]
