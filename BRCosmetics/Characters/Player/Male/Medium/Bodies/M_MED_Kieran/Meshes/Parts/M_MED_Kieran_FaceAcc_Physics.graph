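# Physics asset for the Kieran male medium character's face accessory part, giving a front hair strand jointed physics on the head.
# Capsules cover a four-bone chain from base_hair_front to hair_front_F_3, plus two larger overlapping capsules shaping the head.
# Constraints link each bone to its parent: the base rotates freely against the head, while the links below it have limited swing and locked twist.
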PhysicsAsset /BRCosmetics/Characters/Player/Male/Medium/Bodies/M_MED_Kieran/Meshes/Parts/M_MED_Kieran_FaceAcc_Physics
bodies {
 base_hair_front [{shapeType=Capsule, length=0.505, radius=0.505, shapeName=base_hair_front_capsule, rotation=<rot 0 -90 0>}]
 hair_front_F_1 [{shapeType=Capsule, length=1, radius=0.5, shapeName=hair_front_F_1_capsule, rotation=<rot 90 0 0>}]
 hair_front_F_2 [{shapeType=Capsule, length=2, radius=1, shapeName=hair_front_F_2_capsule, center=<-0.1129 0 0>, rotation=<rot 90 0 0>}]
 hair_front_F_3 [{shapeType=Capsule, length=4, radius=1, shapeName=hair_front_F_3_capsule, center=<1.6761 0 0>, rotation=<rot 90 0 0>}]
 head [
  {shapeType=Capsule, length=12, radius=6, shapeName=head_capsule, center=<-0.7381 3.5409 -0.7998>, rotation=<rot 89.0107 169.9156 178.0022>}
  {shapeType=Capsule, length=12, radius=6, shapeName=head_capsule, center=<-1.2363 -0.6064 0.4322>, rotation=<rot 90 -7.125 1.4009>}
 ]
}
constraints [
 {bone1Name=hair_front_F_2, bone2Name=hair_front_F_1, swing1LimitDegrees=20, swing2LimitDegrees=20, twistLimitDegrees=0, twistMotion=Locked}
 {bone1Name=hair_front_F_3, bone2Name=hair_front_F_2, swing1LimitDegrees=20, swing2LimitDegrees=20, twistLimitDegrees=0, twistMotion=Locked}
 {bone1Name=hair_front_F_1, bone2Name=base_hair_front, swing1LimitDegrees=15, swing2LimitDegrees=40, twistLimitDegrees=0, twistMotion=Locked}
 {
  bone1Name=base_hair_front
  bone2Name=head
  swing1LimitDegrees=45
  swing1Motion=Free
  swing2LimitDegrees=45
  swing2Motion=Free
  twistLimitDegrees=45
  twistMotion=Free
 }
]
